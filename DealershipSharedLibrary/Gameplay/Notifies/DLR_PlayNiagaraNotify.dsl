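; Dealership anim notify that configures its spawned Niagara effect for cosmetics.
; On notify it sets absolute rotation, flags front-end previews, and passes a paint variant from the DLR_Cosmetics_Base owner, or a debug override that also writes mesh custom primitive data.
; A separate ApplyWaterParameters function, not called in this graph, optionally feeds the player pawn's water state, depth, surface plane and water body to Niagara, with debug prints.
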(fn ApplyWaterParameters
  (params (PlayerPawn :type Fort%20Player%20Pawn%20Athena%20Object%20Reference) (NiagaraSystem :type Niagara%20Particle%20System%20Component%20Object%20Reference))
  (bind _athenaplayerpawn (LocalVariables|GetAthenaPlayerPawn))
  (bind _returnvalue (WaterBody|GetCurrentWaterBody _athenaplayerpawn))
  (bind _isinwater (LocalVariables|GetisInWater))
  (bind _waterbodycomponent (Variables|/Script/Water.WaterBody|GetWaterBodyComponent _returnvalue))
  (bind _asplayer_pawn_athena (Utilities|Casting|CastToPlayerPawn_Athena PlayerPawn))
  (bind _output_get (LocalVariables|SetAthenaPlayerPawn _asplayer_pawn_athena))
  (bind _asfort_movement_comp_character (Utilities|Casting|CastToFortMovementComp_Character (Variables|/Script/Engine.Character|GetCharacterMovement _output_get)))
  (LocalVariables|SetFortMovementComp _asfort_movement_comp_character)
  (if (Variables|Niagara|GetSet%20in%20Water%20Niagara%20Variable)
    (Utilities|FlowControl|Sequence
      (:then_0
        ;[1 1 1] qq
        (Development|PrintString (Utilities|String|Append "Is In Water: " (Utilities|String|ToString(Boolean) (or (Utilities|IsValid (WaterBody|GetCurrentWaterBody (LocalVariables|GetAthenaPlayerPawn)) :!class /Script/Engine.KismetSystemLibrary) (WaterBody|IsSurfaceSwimming (LocalVariables|GetAthenaPlayerPawn))))) (Variables|Niagara|GetPrint%20Debug) (Variables|Niagara|GetPrint%20Debug) "(R=0.000000,G=1.000000,B=0.377983,A=1.000000)"))
      (:then_1
        ;[1 1 1 1 40] Set local variables
        (bind _output_get_2 (LocalVariables|SetisInWater (or (Utilities|IsValid (WaterBody|GetCurrentWaterBody (LocalVariables|GetAthenaPlayerPawn)) :!class /Script/Engine.KismetSystemLibrary) (WaterBody|IsSurfaceSwimming (LocalVariables|GetAthenaPlayerPawn)))))
        (if _output_get_2
          (LocalVariables|SetwaveParamTextureOffset (Variables|/Script/Water.WaterBody|GetWaterBodyIndex (WaterBody|GetCurrentWaterBody (LocalVariables|GetAthenaPlayerPawn))))
          (LocalVariables|SetwaterImmersionDepth (WaterBody|GetWaterImmersionDepth (LocalVariables|GetAthenaPlayerPawn)))
          (LocalVariables|SetwaterSurfaceLocation (Swimming|GetWaterSurfaceLocation (LocalVariables|GetFortMovementComp)))
          (LocalVariables|SetwaterSurfaceNormal (Swimming|GetWaterSurfaceNormal (LocalVariables|GetFortMovementComp)))))
      (:then_2
        ;[1 1 1 1 40] Debug
        (Development|PrintString (Utilities|String|Append "In Water : " (Utilities|String|ToString(Boolean) (LocalVariables|GetisInWater))) (Variables|Niagara|GetPrint%20Debug) (Variables|Niagara|GetPrint%20Debug) "(R=0.000000,G=1.000000,B=0.419013,A=1.000000)")
        (Development|PrintString (Utilities|String|JoinStringArray (Utilities|Array|MakeArray (Utilities|String|Append "Depth : " (Utilities|String|ToString(Float) (LocalVariables|GetwaterImmersionDepth))) (Utilities|String|Append "Water Location : " (Utilities|String|ToString(Vector) (LocalVariables|GetwaterSurfaceLocation))) (Utilities|String|Append "Water Normal : " (Utilities|String|ToString(Vector) (LocalVariables|GetwaterSurfaceNormal))) (Utilities|String|Append "Wave Tex Offset : " (Utilities|String|ToString(Integer) (LocalVariables|GetwaveParamTextureOffset)))) " || ") (Variables|Niagara|GetPrint%20Debug) (Variables|Niagara|GetPrint%20Debug))
        (Development|PrintString (Utilities|String|Append "Water Body Component: " (Utilities|GetDisplayName (Variables|/Script/Water.WaterBody|GetWaterBodyComponent (WaterBody|GetCurrentWaterBody (LocalVariables|GetAthenaPlayerPawn))))) (Variables|Niagara|GetPrint%20Debug) (Variables|Niagara|GetPrint%20Debug) "(R=0.000000,G=0.175929,B=0.300000,A=1.000000)"))
      (:then_3
        (if _isinwater
          ;[0 0.4523 1 0.138 80] Set Niagara Water User Variables
          (Niagara|SetNiagaraVariable(Bool) (LocalVariables|GetNiagaraSystem) "User.bIsInWater" _isinwater)
          (Niagara|SetNiagaraVariable(Float) (LocalVariables|GetNiagaraSystem) "User.WaterBodyIndex" (Math|Conversions|ToFloat(Integer) (LocalVariables|GetwaveParamTextureOffset)))
          (Niagara|SetNiagaraVariable(Float) (LocalVariables|GetNiagaraSystem) "User.WaterDepth" (LocalVariables|GetwaterImmersionDepth))
          (Niagara|SetNiagaraVariable(Vector3) (LocalVariables|GetNiagaraSystem) "User.WaterPlaneLocation" (LocalVariables|GetwaterSurfaceLocation))
          (Niagara|SetNiagaraVariable(Vector3) (LocalVariables|GetNiagaraSystem) "User.WaterPlaneNormal" (LocalVariables|GetwaterSurfaceNormal))
          (Utilities|IsValid _waterbodycomponent
            (:Is%20Valid
              (Water|SetWaterBodyComponent (LocalVariables|GetNiagaraSystem) "WaterBodyActor" _waterbodycomponent))))))))

(fn Received_Notify
  (params (MeshComp :type Skeletal%20Mesh%20Component%20Object%20Reference) (Animation :type Anim%20Sequence%20Base%20Object%20Reference) (EventReference :type Anim%20Notify%20Event%20Reference%20Structure%20%28by%20ref%29))
  (bind _print_debug (Variables|Niagara|GetPrint%20Debug))
  (bind _returnvalue (AnimNotify|GetSpawnedEffect))
  (bind _asniagara_particle_system_component (Utilities|Casting|CastToNiagaraComponent _returnvalue)
    (:then
      (LocalVariables|SetNiagaraSystem _asniagara_particle_system_component)
      (Utilities|FlowControl|Sequence
        (:then_0
          (Variables|/Script/Engine.SceneComponent|SetbAbsoluteRotation (Variables|Default|GetAbsolute%20Rotation) (LocalVariables|GetNiagaraSystem)))
        (:then_1
          (bind _output_get (LocalVariables|SetisFrontEnd (Reactive|Frontend|IsCosmeticObjectDisplayedinFrontendContext (Variables|GetMeshComp))))
          (if _output_get
            (Development|PrintString "Setting Front End Variable" _print_debug _print_debug "(R=1.000000,G=0.000000,B=0.083156,A=1.000000)" 1e+01)
            (Niagara|SetNiagaraVariable(Bool) (LocalVariables|GetNiagaraSystem) "User.bIsFrontEndPreview" (LocalVariables|GetisFrontEnd))
            (else
              (Development|PrintString "Not setting Front End Variable" _print_debug _print_debug "(R=1.000000,G=0.000000,B=0.083156,A=1.000000)" 1e+01))))
        (:then_2
          (if (Variables|Dealership|GetDebug%20Painted%20Color)
            (Rendering|Material|SetCustomPrimitiveDataFloat (Variables|GetMeshComp) 0 (Math|Conversions|ToFloat(Byte) (Variables|Dealership|GetPaint%20Variant)))
            (Niagara|SetNiagaraVariable(Int32) (LocalVariables|GetNiagaraSystem) "ColorVariantID" (Math|Conversions|ToInteger(Byte) (Variables|Dealership|GetPaint%20Variant)))
            (else
              (bind _asdlr_cosmetics_base (Utilities|Casting|CastToDLR_Cosmetics_Base (Components|GetOwner (Variables|GetMeshComp)))
                (:then
                  (Niagara|SetNiagaraVariable(Int32) (LocalVariables|GetNiagaraSystem) "ColorVariantID" (Math|Conversions|ToInteger(Byte) (Variables|/DealershipSharedLibrary/Gameplay/DLR_Cosmetics_Base.DLR_Cosmetics_Base_C|GetPaint%20Variant _asdlr_cosmetics_base))))
                (:CastFailed
                  (return false))))))
        (:then_3
          (return false))))
    (:CastFailed
      (return false))))
